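; Behavior tree injection evaluator for the AxisThrow AI that decides whether a custom intro can start, judging by its name.
; On enter, it gets the controlled pawn from the tree component's owning controller and caches it.
; The starting condition passes only if that pawn is valid and its owned gameplay tags match a configurable success tag query.
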(event EventOnEnter
  (params (OwnerComp :type Behavior%20Tree%20Component%20Object%20Reference))
  (bind _ascontroller (Utilities|Casting|CastToController (Components|GetOwner OwnerComp)))
  (Variables|Default|SetPawn (Pawn|GetControlledPawn _ascontroller)))

(fn EvaluateStartingConditions
  (params (OwnerComp :type Behavior%20Tree%20Component%20Object%20Reference))
  (bind _pawn (Variables|Default|GetPawn))
  (Utilities|IsValid _pawn
    (:Is%20Valid
      (bind _returnvalue (GameplayTags|GetOwnedGameplayTags _pawn :!class /Script/GameplayTags.GameplayTagAssetInterface))
      (return (GameplayTags|DoesContainerMatchTagQuery _returnvalue (Variables|Default|GetTagQueryForSuccessfulEval))))
    (:Is%20Not%20Valid
      (return false))))
